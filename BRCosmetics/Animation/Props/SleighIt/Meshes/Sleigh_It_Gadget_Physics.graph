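# Physics asset for the Sleigh It gadget, a cosmetic animation prop mesh.
# Five identical capsules on the lamaA to lamaE base bones are each constrained to a small lama_master capsule with 45-degree swing and twist limits.
# A large root sphere and a stringMaster capsule add further collision shapes, with no constraints linking them to the other bodies.
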PhysicsAsset /BRCosmetics/Animation/Props/SleighIt/Meshes/Sleigh_It_Gadget_Physics
bodies {
 lamaA_base [{shapeType=Capsule, length=7.537, radius=4.7841, center=<0.0025 2.0717 0.0156>, rotation=<rot 0 -13.502 -89.8385>}]
 lamaB_base [{shapeType=Capsule, length=7.537, radius=4.7841, center=<0.0025 2.0717 0.0156>, rotation=<rot 0 -13.502 -89.8385>}]
 lamaC_base [{shapeType=Capsule, length=7.537, radius=4.7841, center=<0.0025 2.0717 0.0156>, rotation=<rot 0 -13.502 -89.8385>}]
 lamaD_base [{shapeType=Capsule, length=7.537, radius=4.7841, center=<0.0025 2.0717 0.0156>, rotation=<rot 0 -13.502 -89.8385>}]
 lamaE_base [{shapeType=Capsule, length=7.537, radius=4.7841, center=<0.0025 2.0717 0.0156>, rotation=<rot 0 -13.502 -89.8385>}]
 lama_master [{shapeType=Capsule, length=0.505, radius=0.505, rotation=<rot 0 -90 0>}]
 root [{shapeType=Sphere, radius=30.5, shapeName=root_sphere, center=<0 0 -12.8588>}]
 stringMaster [{shapeType=Capsule, length=16.8559, radius=6.3721, center=<-0.1716 15.0773 2.4764>, rotation=<rot 0 -179.3702 -80.2232>}]
}
constraints [
 {bone1Name=lamaA_base, bone2Name=lama_master, swing1LimitDegrees=45, swing2LimitDegrees=45, twistLimitDegrees=45}
 {bone1Name=lamaB_base, bone2Name=lama_master, swing1LimitDegrees=45, swing2LimitDegrees=45, twistLimitDegrees=45}
 {bone1Name=lamaC_base, bone2Name=lama_master, swing1LimitDegrees=45, swing2LimitDegrees=45, twistLimitDegrees=45}
 {bone1Name=lamaD_base, bone2Name=lama_master, swing1LimitDegrees=45, swing2LimitDegrees=45, twistLimitDegrees=45}
 {bone1Name=lamaE_base, bone2Name=lama_master, swing1LimitDegrees=45, swing2LimitDegrees=45, twistLimitDegrees=45}
]
